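# Physics asset for the skydive variant of the male medium PowerFarmer_Damage body, simulating its jacket and belts.
# Capsules on pelvis, spine and thighs form the body, linked by free-motion joints.
# Five jacket chains (front and back on both sides, plus left) hang from MSTR bodies on spine_02, mostly with tight swing and twist limits.
# Left and right belt chains branch off the lower front jacket segments.
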PhysicsAsset /BRCosmetics/Characters/Player/Male/Medium/Bodies/M_MED_PowerFarmer_Damage/Meshes/M_MED_PowerFarmer_Damage_Physics_SkyDive
bodies {
 dyn_belt_l_01 [{shapeType=Capsule, length=3.8755, radius=2.1245, shapeName=dyn_belt_l_01_capsule, center=<2 2 0>, rotation=<rot 90 0 0>}]
 dyn_belt_l_02 [{shapeType=Capsule, length=3, radius=2, shapeName=dyn_belt_l_02_capsule, center=<2 2 0>, rotation=<rot 90 0 0>}]
 dyn_belt_l_03 [{shapeType=Capsule, length=3, radius=2, shapeName=dyn_belt_l_03_capsule, center=<2 2 0>, rotation=<rot 90 0 0>}]
 dyn_belt_l_04 [{shapeType=Capsule, length=2.6989, radius=2.3011, shapeName=dyn_belt_l_04_capsule, center=<2 2 0>, rotation=<rot -3.2116 3.8344 -50.1075>}]
 dyn_belt_r_01 [{shapeType=Capsule, length=2, radius=2, shapeName=dyn_belt_r_01_capsule, center=<2 2 0>, rotation=<rot 90 0 0>}]
 dyn_belt_r_02 [{shapeType=Capsule, length=2, radius=2, shapeName=dyn_belt_r_02_capsule, center=<2 2 0>, rotation=<rot 90 0 0>}]
 dyn_belt_r_03 [{shapeType=Capsule, length=2, radius=2, shapeName=dyn_belt_r_03_capsule, center=<2 2 0>, rotation=<rot 90 0 0>}]
 dyn_jacket_bk_l_01 [{shapeType=Capsule, length=10, radius=3.5, shapeName=dyn_jacket_bk_l_01_capsule, center=<2 0 2>, rotation=<rot 0 0 90>}]
 dyn_jacket_bk_l_02 [{shapeType=Capsule, length=10, radius=3.6, shapeName=dyn_jacket_bk_l_02_capsule, center=<2 0 2>, rotation=<rot 0 0 90>}]
 dyn_jacket_bk_l_03 [{shapeType=Capsule, length=10, radius=3.6, shapeName=dyn_jacket_bk_l_03_capsule, center=<2 0 2>, rotation=<rot 0 0 93.8208>}]
 dyn_jacket_bk_l_04 [{shapeType=Capsule, length=10, radius=4.3, shapeName=dyn_jacket_bk_l_04_capsule, center=<2 0 2>, rotation=<rot 0 0 97.9263>}]
 dyn_jacket_bk_l_MSTR [{shapeType=Capsule, length=0.505, radius=0.505, shapeName=dyn_jacket_bk_l_MSTR_capsule, rotation=<rot 0 -90 0>}]
 dyn_jacket_bk_r_01 [{
   shapeType=Capsule
   length=0.505
   radius=0.505
   shapeName=dyn_jacket_bk_r_01_capsule
   center=<6.4389 3.7142 -0.5525>
   rotation=<rot 39.694 -90 0.0582>
  }]
 dyn_jacket_bk_r_02 [{shapeType=Capsule, length=10, radius=4, shapeName=dyn_jacket_bk_r_02_capsule, center=<2 0 2>, rotation=<rot 0 0 90>}]
 dyn_jacket_bk_r_03 [{shapeType=Capsule, length=10, radius=3.3, shapeName=dyn_jacket_bk_r_03_capsule, center=<2 0 2>, rotation=<rot 0 0 90>}]
 dyn_jacket_bk_r_04 [{shapeType=Capsule, length=10, radius=3, shapeName=dyn_jacket_bk_r_04_capsule, center=<2 0 2>, rotation=<rot 0 0 90>}]
 dyn_jacket_bk_r_MSTR [{shapeType=Capsule, length=0.505, radius=0.505, shapeName=dyn_jacket_bk_r_MSTR_capsule, rotation=<rot 0 -90 0>}]
 dyn_jacket_fr_l_01 [{shapeType=Capsule, length=10, radius=3, shapeName=dyn_jacket_fr_l_01_capsule, center=<2 0 1.5>, rotation=<rot 0 0 90>}]
 dyn_jacket_fr_l_02 [{shapeType=Capsule, length=10, radius=3, shapeName=dyn_jacket_fr_l_02_capsule, center=<2 0 1.5>, rotation=<rot 0 0 90>}]
 dyn_jacket_fr_l_03 [{shapeType=Capsule, length=10, radius=3, shapeName=dyn_jacket_fr_l_03_capsule, center=<2 0 1.5>, rotation=<rot 0 0 90>}]
 dyn_jacket_fr_l_04 [{shapeType=Capsule, length=11.1078, radius=1.8922, shapeName=dyn_jacket_fr_l_04_capsule, center=<2 0 -0.1186>, rotation=<rot 0 0 90>}]
 dyn_jacket_fr_l_MSTR [{shapeType=Capsule, length=0.505, radius=0.505, shapeName=dyn_jacket_fr_l_MSTR_capsule, rotation=<rot 0 -90 0>}]
 dyn_jacket_fr_r_01 [{shapeType=Capsule, length=0.505, radius=0.505, shapeName=dyn_jacket_fr_r_01_capsule, rotation=<rot 0 -90 0>}]
 dyn_jacket_fr_r_02 [{shapeType=Capsule, length=10, radius=3, shapeName=dyn_jacket_fr_r_02_capsule, center=<2 0 2>, rotation=<rot 0 0 90>}]
 dyn_jacket_fr_r_03 [{shapeType=Capsule, length=10, radius=3, shapeName=dyn_jacket_fr_r_03_capsule, center=<2 0 2>, rotation=<rot 0 0 90>}]
 dyn_jacket_fr_r_04 [{shapeType=Capsule, length=10, radius=3, shapeName=dyn_jacket_fr_r_04_capsule, center=<2 0 2>, rotation=<rot 0 0 90>}]
 dyn_jacket_fr_r_MSTR [{shapeType=Capsule, length=0.505, radius=0.505, shapeName=dyn_jacket_fr_r_MSTR_capsule, rotation=<rot 0 -90 0>}]
 dyn_jacket_l_01 [{shapeType=Capsule, length=10, radius=3, shapeName=dyn_jacket_l_01_capsule, center=<2 0 1>, rotation=<rot 0 0 90>}]
 dyn_jacket_l_02 [{shapeType=Capsule, length=10, radius=3, shapeName=dyn_jacket_l_02_capsule, center=<2 0 1>, rotation=<rot 0 0 90>}]
 dyn_jacket_l_03 [{shapeType=Capsule, length=10, radius=3, shapeName=dyn_jacket_l_03_capsule, center=<2 0 1>, rotation=<rot 0 0 90>}]
 dyn_jacket_l_04 [{shapeType=Capsule, length=10, radius=3, shapeName=dyn_jacket_l_04_capsule, center=<2 0 1>, rotation=<rot 0 0 90>}]
 dyn_jacket_l_MSTR [{shapeType=Capsule, length=0.505, radius=0.505, shapeName=dyn_jacket_l_MSTR_capsule, rotation=<rot 0 -90 0>}]
 pelvis [
  {shapeType=Capsule, length=13.7649, radius=8.5753, shapeName=pelvis_capsule, center=<0 -3.0344 0>}
  {shapeType=Capsule, length=22.1715, radius=9.3548, shapeName=pelvis_capsule, center=<1.4259 12.5328 -0.1648>, rotation=<rot 0 -0.0001 -95>}
 ]
 spine_01 [{shapeType=Capsule, length=13.611, radius=13.389, shapeName=spine_01_capsule, center=<0 1.3425 0>, rotation=<rot 90 0 0>}]
 spine_02 [{shapeType=Capsule, length=14.4837, radius=12.5163, shapeName=spine_02_capsule, center=<0 0.8057 0>, rotation=<rot 90 0 0>}]
 thigh_l [{shapeType=Capsule, length=35.6912, radius=8.6835, shapeName=thigh_l_capsule, center=<-14.211 0 0>, rotation=<rot 90 0 0>}]
 thigh_r [{shapeType=Capsule, length=33.1811, radius=9.8189, shapeName=thigh_r_capsule, center=<19.0408 0.913 0>, rotation=<rot 90 166.7373 166.7376>}]
}
constraints [
 {
  bone1Name=spine_02
  bone2Name=spine_01
  swing1LimitDegrees=45
  swing1Motion=Free
  swing2LimitDegrees=45
  swing2Motion=Free
  twistLimitDegrees=45
  twistMotion=Free
 }
 {
  bone1Name=spine_01
  bone2Name=pelvis
  swing1LimitDegrees=45
  swing1Motion=Free
  swing2LimitDegrees=45
  swing2Motion=Free
  twistLimitDegrees=45
  twistMotion=Free
 }
 {
  bone1Name=thigh_l
  bone2Name=pelvis
  swing1LimitDegrees=45
  swing1Motion=Free
  swing2LimitDegrees=45
  swing2Motion=Free
  twistLimitDegrees=45
  twistMotion=Free
 }
 {
  bone1Name=dyn_jacket_bk_r_01
  bone2Name=dyn_jacket_bk_r_MSTR
  swing1LimitDegrees=45
  swing1Motion=Free
  swing2LimitDegrees=45
  swing2Motion=Free
  twistLimitDegrees=45
  twistMotion=Free
 }
 {bone1Name=dyn_jacket_bk_r_02, bone2Name=dyn_jacket_bk_r_01, swing1LimitDegrees=10, swing2LimitDegrees=5, twistLimitDegrees=2}
 {bone1Name=dyn_jacket_bk_r_03, bone2Name=dyn_jacket_bk_r_02, swing1LimitDegrees=10, swing2LimitDegrees=5, twistLimitDegrees=2}
 {bone1Name=dyn_jacket_bk_r_04, bone2Name=dyn_jacket_bk_r_03, swing1LimitDegrees=10, swing2LimitDegrees=5, twistLimitDegrees=2}
 {
  bone1Name=dyn_jacket_fr_r_01
  bone2Name=dyn_jacket_fr_r_MSTR
  swing1LimitDegrees=45
  swing1Motion=Free
  swing2LimitDegrees=45
  swing2Motion=Free
  twistLimitDegrees=45
  twistMotion=Free
 }
 {bone1Name=dyn_jacket_fr_r_02, bone2Name=dyn_jacket_fr_r_01, swing1LimitDegrees=20, swing2LimitDegrees=20, twistLimitDegrees=5}
 {bone1Name=dyn_jacket_fr_r_03, bone2Name=dyn_jacket_fr_r_02, swing1LimitDegrees=20, swing2LimitDegrees=20, twistLimitDegrees=5}
 {bone1Name=dyn_jacket_fr_r_04, bone2Name=dyn_jacket_fr_r_03, swing1LimitDegrees=15, swing2LimitDegrees=10, twistLimitDegrees=2}
 {bone1Name=dyn_jacket_l_04, bone2Name=dyn_jacket_l_03, swing1LimitDegrees=10, swing2LimitDegrees=5, twistLimitDegrees=2}
 {bone1Name=dyn_jacket_l_01, bone2Name=dyn_jacket_l_MSTR, swing1LimitDegrees=10, swing2LimitDegrees=5, twistLimitDegrees=2}
 {bone1Name=dyn_jacket_l_02, bone2Name=dyn_jacket_l_01, swing1LimitDegrees=10, swing2LimitDegrees=5, twistLimitDegrees=2}
 {bone1Name=dyn_jacket_l_03, bone2Name=dyn_jacket_l_02, swing1LimitDegrees=10, swing2LimitDegrees=5, twistLimitDegrees=2}
 {bone1Name=dyn_jacket_bk_l_04, bone2Name=dyn_jacket_bk_l_03, swing1LimitDegrees=10, swing2LimitDegrees=5, twistLimitDegrees=2}
 {bone1Name=dyn_jacket_bk_l_03, bone2Name=dyn_jacket_bk_l_02, swing1LimitDegrees=10, swing2LimitDegrees=5, twistLimitDegrees=2}
 {bone1Name=dyn_jacket_bk_l_02, bone2Name=dyn_jacket_bk_l_01, swing1LimitDegrees=10, swing2LimitDegrees=5, twistLimitDegrees=2}
 {bone1Name=dyn_jacket_bk_l_01, bone2Name=dyn_jacket_bk_l_MSTR, swing1LimitDegrees=10, swing2LimitDegrees=5, twistLimitDegrees=2}
 {
  bone1Name=dyn_jacket_bk_l_MSTR
  bone2Name=spine_02
  swing1LimitDegrees=45
  swing1Motion=Free
  swing2LimitDegrees=45
  swing2Motion=Free
  twistLimitDegrees=45
  twistMotion=Free
 }
 {
  bone1Name=dyn_jacket_bk_r_MSTR
  bone2Name=spine_02
  swing1LimitDegrees=45
  swing1Motion=Free
  swing2LimitDegrees=45
  swing2Motion=Free
  twistLimitDegrees=45
  twistMotion=Free
 }
 {
  bone1Name=dyn_jacket_fr_l_MSTR
  bone2Name=spine_02
  swing1LimitDegrees=45
  swing1Motion=Free
  swing2LimitDegrees=45
  swing2Motion=Free
  twistLimitDegrees=45
  twistMotion=Free
 }
 {
  bone1Name=dyn_jacket_fr_r_MSTR
  bone2Name=spine_02
  swing1LimitDegrees=45
  swing1Motion=Free
  swing2LimitDegrees=45
  swing2Motion=Free
  twistLimitDegrees=45
  twistMotion=Free
 }
 {
  bone1Name=dyn_jacket_l_MSTR
  bone2Name=spine_02
  swing1LimitDegrees=45
  swing1Motion=Free
  swing2LimitDegrees=45
  swing2Motion=Free
  twistLimitDegrees=45
  twistMotion=Free
 }
 {bone1Name=dyn_jacket_fr_l_01, bone2Name=dyn_jacket_fr_l_MSTR, swing1LimitDegrees=15, swing2LimitDegrees=20, twistLimitDegrees=2}
 {bone1Name=dyn_jacket_fr_l_02, bone2Name=dyn_jacket_fr_l_01, swing1LimitDegrees=15, swing2LimitDegrees=20, twistLimitDegrees=2}
 {bone1Name=dyn_jacket_fr_l_03, bone2Name=dyn_jacket_fr_l_02, swing1LimitDegrees=25, swing2LimitDegrees=20, twistLimitDegrees=2}
 {bone1Name=dyn_jacket_fr_l_04, bone2Name=dyn_jacket_fr_l_03, swing1LimitDegrees=25, swing2LimitDegrees=20, twistLimitDegrees=2}
 {bone1Name=dyn_belt_l_01, bone2Name=dyn_jacket_fr_l_04, swing1LimitDegrees=2, swing2LimitDegrees=2, twistLimitDegrees=2}
 {bone1Name=dyn_belt_l_02, bone2Name=dyn_belt_l_01, swing1LimitDegrees=20, swing2LimitDegrees=10, twistLimitDegrees=1}
 {bone1Name=dyn_belt_l_03, bone2Name=dyn_belt_l_02, swing1LimitDegrees=20, swing2LimitDegrees=10, twistLimitDegrees=1}
 {bone1Name=dyn_belt_l_04, bone2Name=dyn_belt_l_03, swing1LimitDegrees=20, swing2LimitDegrees=10, twistLimitDegrees=1}
 {
  bone1Name=thigh_r
  bone2Name=pelvis
  swing1LimitDegrees=45
  swing1Motion=Free
  swing2LimitDegrees=45
  swing2Motion=Free
  twistLimitDegrees=45
  twistMotion=Free
 }
 {bone1Name=dyn_belt_r_01, bone2Name=dyn_jacket_fr_r_03, swing1LimitDegrees=5, swing2LimitDegrees=5, twistLimitDegrees=1}
 {bone1Name=dyn_belt_r_02, bone2Name=dyn_belt_r_01, swing1LimitDegrees=5, swing2LimitDegrees=5, twistLimitDegrees=1}
 {bone1Name=dyn_belt_r_03, bone2Name=dyn_belt_r_02, swing1LimitDegrees=5, swing2LimitDegrees=5, twistLimitDegrees=1}
]
