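# Post-pose AnimBlueprint for the PowerFarmer body cosmetic, adding secondary motion to jacket, zipper, belt and bandolier bones.
# The AnimGraph copies the attached parent's pose, then chains weighted bone constraints, AnimDynamics, a Control Rig and two RigidBody nodes, ending with jacket constraints.
# The rig and both physics assets (default, skydive) come from the PowerFarmer_Damage variant.
# The EventGraph derives float flags for stance, movement direction, slide and weapon type, many feeding the rig.
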AnimBlueprint /BRCosmetics/Characters/Player/Male/Medium/Bodies/M_MED_PowerFarmer/Meshes/M_MED_PowerFarmer_AnimBP
target_skeleton=&/M_MED_PowerFarmer_Skeleton
graphs [
 AnimGraph {nodes [
   n1 = "Constraint - dyn_zipper_pocket (PPPP)"!(
    ComponentPose=Local To Component(LocalPose=Copy Pose From Mesh() {bUseAttachedParent=true})
    Alpha=1
    ConstraintWeights_0=0.4598
    ConstraintWeights_1:float=0.4306
    ConstraintWeights_2:float=0.1035
    ConstraintWeights_3:float=0.0061
   ) -> (Pose) {
    alphaBoolBlend {blendOption=Linear}
    boneToModify {boneName=dyn_zipper_pocket}
    constraintSetup [
     {targetBone {boneName=dyn_jacket_fr_l_01}}
     {targetBone {boneName=dyn_jacket_fr_l_02}}
     {targetBone {boneName=dyn_jacket_fr_l_03}}
     {targetBone {boneName=dyn_jacket_fr_l_04}}
    ]
    constraintWeights [1,1,1,1]
   }
   n2 = "Constraint - dyn_zipper_jacket (PPPP)"!(
    ComponentPose=@n1
    Alpha=1
    ConstraintWeights_0=0.0079
    ConstraintWeights_1:float=0.0424
    ConstraintWeights_2:float=0.4814
    ConstraintWeights_3:float=0.4684
   ) -> (Pose) {
    alphaBoolBlend {blendOption=Linear}
    boneToModify {boneName=dyn_zipper_jacket}
    constraintSetup [
     {targetBone {boneName=dyn_jacket_fr_l_01}}
     {targetBone {boneName=dyn_jacket_fr_l_02}}
     {targetBone {boneName=dyn_jacket_fr_l_03}}
     {targetBone {boneName=dyn_jacket_fr_l_04}}
    ]
    constraintWeights [1,1,1,1]
   }
   n3 = "Constraint - dyn_zipper_lace_01 (PPPPPPPP)"!(
    ComponentPose=@n2
    Alpha=1
    ConstraintWeights_0=0.209
    ConstraintWeights_1:float=0.0135
    ConstraintWeights_2:float=0.3409
    ConstraintWeights_3:float=0.1135
    ConstraintWeights_4:float=0.1969
    ConstraintWeights_5:float=0.1241
    ConstraintWeights_6:float=0.0013
    ConstraintWeights_7:float=0.0009
   ) -> (Pose) {
    alphaBoolBlend {blendOption=Linear}
    boneToModify {boneName=dyn_zipper_lace_01}
    constraintSetup [
     {targetBone {boneName=upperarm_twist_01_l}}
     {targetBone {boneName=spine_02}}
     {targetBone {boneName=spine_04}}
     {targetBone {boneName=spine_05}}
     {targetBone {boneName=clavicle_l}}
     {targetBone {boneName=spine_03}}
     {targetBone {boneName=neck_02}}
     {targetBone {boneName=upperarm_twist_02_l}}
    ]
    constraintWeights [1,1,1,1,1,1,1,1]
   }
   n4 = "Constraint - dyn_belt_l_01 (PPPP)"!(
    ComponentPose=@n3
    Alpha=0.2
    ConstraintWeights_0=0.034
    ConstraintWeights_1:float=0.1491
    ConstraintWeights_2:float=0.4414
    ConstraintWeights_3:float=0.3755
   ) -> (Pose) {
    alphaBoolBlend {blendOption=Linear}
    boneToModify {boneName=dyn_belt_l_01}
    constraintSetup [
     {targetBone {boneName=dyn_jacket_fr_l_01}}
     {targetBone {boneName=dyn_jacket_fr_l_02}}
     {targetBone {boneName=dyn_jacket_fr_l_03}}
     {targetBone {boneName=dyn_jacket_fr_l_04}}
    ]
    constraintWeights [1,1,1,1]
   }
   n5 = "Constraint - dyn_belt_l_02 (PPPP)"!(
    ComponentPose=@n4
    Alpha=1
    ConstraintWeights_0=0.0487
    ConstraintWeights_1:float=0.2396
    ConstraintWeights_2:float=0.4413
    ConstraintWeights_3:float=0.2704
   ) -> (Pose) {
    alphaBoolBlend {blendOption=Linear}
    boneToModify {boneName=dyn_belt_l_02}
    constraintSetup [
     {targetBone {boneName=dyn_jacket_fr_l_01}}
     {targetBone {boneName=dyn_jacket_fr_l_02}}
     {targetBone {boneName=dyn_jacket_fr_l_03}}
     {targetBone {boneName=dyn_jacket_fr_l_04}}
    ]
    constraintWeights [1,1,1,1]
   }
   n6 = "Constraint - dyn_belt_l_03 (PPPP)"!(
    ComponentPose=@n5
    Alpha=1
    ConstraintWeights_0=0.0521
    ConstraintWeights_1:float=0.2454
    ConstraintWeights_2:float=0.4362
    ConstraintWeights_3:float=0.2663
   ) -> (Pose) {
    alphaBoolBlend {blendOption=Linear}
    boneToModify {boneName=dyn_belt_l_03}
    constraintSetup [
     {targetBone {boneName=dyn_jacket_fr_l_01}}
     {targetBone {boneName=dyn_jacket_fr_l_02}}
     {targetBone {boneName=dyn_jacket_fr_l_03}}
     {targetBone {boneName=dyn_jacket_fr_l_04}}
    ]
    constraintWeights [1,1,1,1]
   }
   n7 = "Constraint - dyn_bandolier_fr (PPPPPPPP)"!(
    ComponentPose=@n6
    Alpha=1
    ConstraintWeights_0=0.0157
    ConstraintWeights_1:float=0.0017
    ConstraintWeights_2:float=0.2447
    ConstraintWeights_3:float=0.4593
    ConstraintWeights_4:float=0.1719
    ConstraintWeights_5:float=0.0877
    ConstraintWeights_6:float=0.0009
    ConstraintWeights_7:float=0.0182
   ) -> (Pose) {
    alphaBoolBlend {blendOption=Linear}
    boneToModify {boneName=dyn_bandolier_fr}
    constraintSetup [
     {targetBone {boneName=upperarm_twist_01_l}}
     {targetBone {boneName=spine_02}}
     {targetBone {boneName=spine_04}}
     {targetBone {boneName=spine_05}}
     {targetBone {boneName=clavicle_l}}
     {targetBone {boneName=spine_03}}
     {targetBone {boneName=neck_01}}
     {targetBone {boneName=neck_02}}
    ]
    constraintWeights [1,1,1,1,1,1,1,1]
   }
   n8 = "Constraint - dyn_bandolier1 (PPPPPPPP)"!(
    ComponentPose=@n7
    Alpha=1
    ConstraintWeights_0=0.0596
    ConstraintWeights_1:float=0.0175
    ConstraintWeights_2:float=0.3787
    ConstraintWeights_3:float=0.3242
    ConstraintWeights_4:float=0.024
    ConstraintWeights_5:float=0.0453
    ConstraintWeights_6:float=0.0291
    ConstraintWeights_7:float=0.1216
   ) -> (Pose) {
    alphaBoolBlend {blendOption=Linear}
    boneToModify {boneName=dyn_bandolier1}
    constraintSetup [
     {targetBone {boneName=upperarm_twist_01_l}}
     {targetBone {boneName=spine_04}}
     {targetBone {boneName=spine_05}}
     {targetBone {boneName=clavicle_l}}
     {targetBone {boneName=spine_03}}
     {targetBone {boneName=neck_01}}
     {targetBone {boneName=head}}
     {targetBone {boneName=neck_02}}
    ]
    constraintWeights [1,1,1,1,1,1,1,1]
   }
   n9 = "Constraint - dyn_bandolier_bk (PPPPPPP)"!(
    ComponentPose=@n8
    Alpha=1
    ConstraintWeights_0=0.0347
    ConstraintWeights_1:float=0.0571
    ConstraintWeights_2:float=0.3811
    ConstraintWeights_3:float=0.3956
    ConstraintWeights_4:float=0.1027
    ConstraintWeights_5:float=0.0019
    ConstraintWeights_6:float=0.027
   ) -> (Pose) {
    alphaBoolBlend {blendOption=Linear}
    boneToModify {boneName=dyn_bandolier_bk}
    constraintSetup [
     {targetBone {boneName=upperarm_twist_01_l}}
     {targetBone {boneName=spine_04}}
     {targetBone {boneName=spine_05}}
     {targetBone {boneName=clavicle_l}}
     {targetBone {boneName=spine_03}}
     {targetBone {boneName=neck_01}}
     {targetBone {boneName=neck_02}}
    ]
    constraintWeights [1,1,1,1,1,1,1]
   }
   n10 = Anim Dynamics - Bone: dyn_belt_r_02(ComponentPose=@n9, Alpha=1, ExternalForce="0,0,0") {
    angularDampingOverride=0.9999
    angularSpringConstant=500
    bAngularSpring=true
    bOverrideAngularDamping=true
    bUseGravityOverride=true
    bUseSphericalLimits=true
    numSolverIterationsPreUpdate=2
    simulationSpace=BoneRelative
    alphaBoolBlend {blendOption=Linear}
    alphaScaleBiasClamp {bClampResult=true, clampMax=0.5}
    boundBone {boneName=dyn_belt_r_02}
    componentAppliedLinearAccClamp=<3500 3500 3500>
    componentLinearAccScale=<1.2 1.2 1.2>
    componentLinearVelScale=<1.2 1.2 1.2>
    gravityOverride {z=-980}
    physicsBodyDefinitions [{
      collisionType=CustomSphere
      sphereCollisionRadius=2
      boundBone {boneName=dyn_belt_r_02}
      boxExtents=<15 10 10>
      constraintSetup {angularLimitsMax=<5 10 30>, angularLimitsMin=<-5 -45 -30>, angularTarget=<0 0 0>}
      localJointOffset=<3 0 0>
     }]
    relativeSpaceBone {boneName=root}
    sphericalLimits [
     {limitRadius=10, drivingBone {boneName=thigh_r}, sphereLocalOffset=<8 0 0>}
     {limitRadius=15, drivingBone {boneName=thigh_twist_01_r}, sphereLocalOffset=<0 0 0>}
    ]
   }
   n11 = Anim Dynamics - Bone: dyn_belt_r_02(ComponentPose=@n10, Alpha=1, ExternalForce="0,0,0") {
    angularDampingOverride=0.9999
    angularSpringConstant=500
    bAngularSpring=true
    bOverrideAngularDamping=true
    bUseGravityOverride=true
    bUseSphericalLimits=true
    numSolverIterationsPreUpdate=2
    simulationSpace=BoneRelative
    alphaBoolBlend {blendOption=Linear}
    alphaScaleBiasClamp {bClampResult=true, clampMax=0.5}
    boundBone {boneName=dyn_belt_r_02}
    componentAppliedLinearAccClamp=<3500 3500 3500>
    componentLinearAccScale=<1.2 1.2 1.2>
    componentLinearVelScale=<1.2 1.2 1.2>
    gravityOverride {z=-980}
    physicsBodyDefinitions [{
      collisionType=CustomSphere
      sphereCollisionRadius=2
      boundBone {boneName=dyn_belt_r_02}
      boxExtents=<15 10 10>
      constraintSetup {angularLimitsMax=<5 10 30>, angularLimitsMin=<-5 -45 -30>, angularTarget=<0 0 0>}
      localJointOffset=<3 0 0>
     }]
    relativeSpaceBone {boneName=root}
    sphericalLimits [
     {limitRadius=11, drivingBone {boneName=thigh_r}, sphereLocalOffset=<8 0 0>}
     {limitRadius=15, drivingBone {boneName=thigh_twist_01_r}, sphereLocalOffset=<0 0 0>}
    ]
   }
   n12 = Anim Dynamics - Bone: dyn_belt_r_03(ComponentPose=@n11, Alpha=1, ExternalForce="0,0,0") {
    angularDampingOverride=0.9999
    angularSpringConstant=500
    bAngularSpring=true
    bOverrideAngularDamping=true
    bUseGravityOverride=true
    bUseSphericalLimits=true
    numSolverIterationsPreUpdate=2
    simulationSpace=BoneRelative
    alphaBoolBlend {blendOption=Linear}
    alphaScaleBiasClamp {bClampResult=true, clampMax=0.5}
    boundBone {boneName=dyn_belt_r_03}
    componentAppliedLinearAccClamp=<3500 3500 3500>
    componentLinearAccScale=<1.2 1.2 1.2>
    componentLinearVelScale=<1.2 1.2 1.2>
    gravityOverride {z=-980}
    physicsBodyDefinitions [{
      collisionType=CustomSphere
      sphereCollisionRadius=2
      boundBone {boneName=dyn_belt_r_03}
      boxExtents=<15 10 10>
      constraintSetup {angularLimitsMax=<5 10 30>, angularLimitsMin=<-5 -45 -30>, angularTarget=<0 0 0>}
      localJointOffset=<3 0 0>
     }]
    relativeSpaceBone {boneName=root}
    sphericalLimits [
     {limitRadius=11, drivingBone {boneName=thigh_r}, sphereLocalOffset=<8 0 0>}
     {limitRadius=15, drivingBone {boneName=thigh_twist_01_r}, sphereLocalOffset=<0 0 0>}
    ]
   }
   n13 = "Constraint - dyn_zipper_r_sleeve (PPPPPPPP)"!(
    ComponentPose=@n12
    Alpha=1
    ConstraintWeights_0=0.0004
    ConstraintWeights_1:float=0.6203
    ConstraintWeights_2:float=0.0112
    ConstraintWeights_3:float=0.001
    ConstraintWeights_4:float=0.1736
    ConstraintWeights_5:float=0.0003
    ConstraintWeights_6:float=0.1931
    ConstraintWeights_7:float=0.0001
   ) -> (Pose) {
    alphaBoolBlend {blendOption=Linear}
    boneToModify {boneName=dyn_zipper_r_sleeve}
    constraintSetup [
     {targetBone {boneName=hand_r}}
     {targetBone {boneName=lowerarm_twist_02_r}}
     {targetBone {boneName=upperarm_r}}
     {targetBone {boneName=pinky_metacarpal_r}}
     {targetBone {boneName=lowerarm_twist_01_r}}
     {targetBone {boneName=upperarm_twist_02_r}}
     {targetBone {boneName=lowerarm_r}}
     {targetBone {boneName=thumb_01_r}}
    ]
    constraintWeights [1,1,1,1,1,1,1,1]
   }
   n14 = Anim Dynamics - Bone: dyn_zipper_r_sleeve(ComponentPose=@n13, Alpha=1, ExternalForce="0,0,0") {
    angularDampingOverride=0.9999
    angularSpringConstant=100
    bAngularSpring=true
    bOverrideAngularDamping=true
    bUseGravityOverride=true
    simulationSpace=BoneRelative
    alphaBoolBlend {blendOption=Linear}
    boundBone {boneName=dyn_zipper_r_sleeve}
    componentAppliedLinearAccClamp=<3500 3500 3500>
    componentLinearAccScale=<0.2 0.5 1>
    componentLinearVelScale=<0.2 0.5 1>
    gravityOverride {z=-980}
    physicsBodyDefinitions [{
      sphereCollisionRadius=0
      boundBone {boneName=dyn_zipper_r_sleeve}
      boxExtents=<5 10 10>
      constraintSetup {angularLimitsMax=<1 2 20>, angularLimitsMin=<-1 -2 -120>, angularTarget=<0 0 0>}
      localJointOffset=<3 0 0>
     }]
    relativeSpaceBone {boneName=root}
   }
   n15 = "Transform (Modify) Bone - Bone: dyn_zipper_r_sleeve"!(ComponentPose=@n14, Alpha=1, Translation="-0.6,-0.3,-0.1") -> (Pose) {
    translationMode=BMM_Replace
    translationSpace=BCS_BoneSpace
    alphaBoolBlend {blendOption=Linear}
    boneToModify {boneName=dyn_zipper_r_sleeve}
    translation=<-0.6 -0.3 -0.1>
   }
   n16 = Anim Dynamics - Bone: dyn_zipper_lace_02(ComponentPose=@n15, Alpha=1, ExternalForce="0,0,0") {
    angularDampingOverride=0.9999
    angularSpringConstant=100
    bAngularSpring=true
    bOverrideAngularDamping=true
    bUseGravityOverride=true
    simulationSpace=BoneRelative
    alphaBoolBlend {blendOption=Linear}
    boundBone {boneName=dyn_zipper_lace_02}
    componentAppliedLinearAccClamp=<3500 3500 3500>
    componentLinearAccScale=<0.2 0.5 1>
    componentLinearVelScale=<0.2 0.5 1>
    gravityOverride {z=-980}
    physicsBodyDefinitions [{
      sphereCollisionRadius=0
      boundBone {boneName=dyn_zipper_lace_02}
      boxExtents=<5 10 10>
      constraintSetup {angularLimitsMax=<1 10 5>, angularLimitsMin=<-1 -2 -60>, angularTarget=<0 0 0>}
      localJointOffset=<3 0 0>
     }]
    relativeSpaceBone {boneName=root}
   }
   n17 = "Constraint - dyn_bandolier_fr1 (PPPPPPPP)"!(
    ComponentPose=@n16
    Alpha=1
    ConstraintWeights_0=0.0427
    ConstraintWeights_1:float=0.1068
    ConstraintWeights_2:float=0.4219
    ConstraintWeights_3:float=0.2961
    ConstraintWeights_4:float=0.0531
    ConstraintWeights_5:float=0.0133
    ConstraintWeights_6:float=0.0058
    ConstraintWeights_7:float=0.0602
   ) -> (Pose) {
    alphaBoolBlend {blendOption=Linear}
    boneToModify {boneName=dyn_bandolier_fr1}
    constraintSetup [
     {targetBone {boneName=upperarm_twist_01_l}}
     {targetBone {boneName=spine_04}}
     {targetBone {boneName=spine_05}}
     {targetBone {boneName=clavicle_l}}
     {targetBone {boneName=spine_03}}
     {targetBone {boneName=neck_01}}
     {targetBone {boneName=head}}
     {targetBone {boneName=neck_02}}
    ]
    constraintWeights [1,1,1,1,1,1,1,1]
   }
   n18 = "Constraint - dyn_bandolier_fr (PPPPPPPP)"(
    ComponentPose=@n17
    Alpha=1
    ConstraintWeights_0=0.0157
    ConstraintWeights_1=0.0017
    ConstraintWeights_2=0.2447
    ConstraintWeights_3=0.4593
    ConstraintWeights_4=0.1719
    ConstraintWeights_5=0.0877
    ConstraintWeights_6=0.0009
    ConstraintWeights_7=0.0182
   ) {
    alphaBoolBlend {blendOption=Linear}
    boneToModify {boneName=dyn_bandolier_fr}
    constraintSetup [
     {targetBone {boneName=upperarm_twist_01_l}}
     {targetBone {boneName=spine_02}}
     {targetBone {boneName=spine_04}}
     {targetBone {boneName=spine_05}}
     {targetBone {boneName=clavicle_l}}
     {targetBone {boneName=spine_03}}
     {targetBone {boneName=neck_01}}
     {targetBone {boneName=neck_02}}
    ]
    constraintWeights [1,1,1,1,1,1,1,1]
   }
   n19 = "Constraint - dyn_bandolier1 (PPPPPPPP)"(
    ComponentPose=@n18
    Alpha=1
    ConstraintWeights_0=0.1012
    ConstraintWeights_1=0.0181
    ConstraintWeights_2=0.2857
    ConstraintWeights_3=0.4483
    ConstraintWeights_4=0.031
    ConstraintWeights_5=0.0269
    ConstraintWeights_6=0.0138
    ConstraintWeights_7=0.0749
   ) {
    alphaBoolBlend {blendOption=Linear}
    boneToModify {boneName=dyn_bandolier1}
    constraintSetup [
     {targetBone {boneName=upperarm_twist_01_l}}
     {targetBone {boneName=spine_04}}
     {targetBone {boneName=spine_05}}
     {targetBone {boneName=clavicle_l}}
     {targetBone {boneName=spine_03}}
     {targetBone {boneName=neck_01}}
     {targetBone {boneName=head}}
     {targetBone {boneName=neck_02}}
    ]
    constraintWeights [1,1,1,1,1,1,1,1]
   }
   n20 = "Constraint - dyn_bandolier_bk2 (PPPPPPPP)"!(
    ComponentPose=@n19
    Alpha=1
    ConstraintWeights_0=0.0871
    ConstraintWeights_1:float=0.0117
    ConstraintWeights_2:float=0.2087
    ConstraintWeights_3:float=0.5987
    ConstraintWeights_4:float=0.0393
    ConstraintWeights_5:float=0.0119
    ConstraintWeights_6:float=0.0025
    ConstraintWeights_7:float=0.0402
   ) -> (Pose) {
    alphaBoolBlend {blendOption=Linear}
    boneToModify {boneName=dyn_bandolier_bk2}
    constraintSetup [
     {targetBone {boneName=upperarm_twist_01_l}}
     {targetBone {boneName=spine_04}}
     {targetBone {boneName=spine_05}}
     {targetBone {boneName=clavicle_l}}
     {targetBone {boneName=spine_03}}
     {targetBone {boneName=neck_01}}
     {targetBone {boneName=head}}
     {targetBone {boneName=neck_02}}
    ]
    constraintWeights [1,1,1,1,1,1,1,1]
   }
   n21 = "Constraint - dyn_bandolier_bk1 (PPPPPPP)"!(
    ComponentPose=@n20
    Alpha=1
    ConstraintWeights_0=0.0163
    ConstraintWeights_1:float=0
    ConstraintWeights_2:float=0.2236
    ConstraintWeights_3:float=0.6605
    ConstraintWeights_4:float=0.0728
    ConstraintWeights_5:float=0
    ConstraintWeights_6:float=0.0268
   ) -> (Pose) {
    alphaBoolBlend {blendOption=Linear}
    boneToModify {boneName=dyn_bandolier_bk1}
    constraintSetup [
     {targetBone {boneName=upperarm_twist_01_l}}
     {targetBone {boneName=spine_04}}
     {targetBone {boneName=spine_05}}
     {targetBone {boneName=clavicle_l}}
     {targetBone {boneName=spine_03}}
     {targetBone {boneName=neck_01}}
     {targetBone {boneName=neck_02}}
    ]
    constraintWeights [1,1,1,1,1,1,1]
   }
   n22 = "Constraint - dyn_bandolier_bk (PPPPPPP)"(
    ComponentPose=@n21
    Alpha=1
    ConstraintWeights_0=0.0215
    ConstraintWeights_1=0.0426
    ConstraintWeights_2=0.4385
    ConstraintWeights_3=0.3828
    ConstraintWeights_4=0.0979
    ConstraintWeights_5=0
    ConstraintWeights_6=0.0167
   ) {
    alphaBoolBlend {blendOption=Linear}
    boneToModify {boneName=dyn_bandolier_bk}
    constraintSetup [
     {targetBone {boneName=upperarm_twist_01_l}}
     {targetBone {boneName=spine_04}}
     {targetBone {boneName=spine_05}}
     {targetBone {boneName=clavicle_l}}
     {targetBone {boneName=spine_03}}
     {targetBone {boneName=upperarm_l}}
     {targetBone {boneName=neck_02}}
    ]
    constraintWeights [1,1,1,1,1,1,1]
   }
   n23 = Control Rig!(
    Source=Component To Local(ComponentPose=@n22)
    Alpha
    Crouch:double=Get Crouch() -> (Crouch:double)
    Jump:double=Get Jump() -> (Jump:double)
    Rifle:double=Get Rifle() -> (Rifle:double)
    moving:double=Get moving() -> (moving:double)
    left:double=Get Left() -> (Left:double)
    right:double=Get Right() -> (Right:double)
    baseball:double=Get Base_Slide() -> (Base_Slide:double)
    forward:double=Get Forward() -> (Forward:double)
    backward:double=Get Backward() -> (Backward:double)
    pistol:double=Get Pistol() -> (Pistol:double)
   ) -> (Pose) {
    alphaBoolBlend {blendOption=Linear}
    controlRigAssetReference {blueprintRigClass=/BRCosmetics/Characters/Player/Male/Medium/Bodies/M_MED_PowerFarmer_Damage/Meshes/M_MED_PowerFarmer_Damage_CtrlRig.M_MED_PowerFarmer_Damage_CtrlRig_C}
   }
   n24 = RigidBody!(ComponentPose=Local To Component(LocalPose=@n23), Alpha, OverrideWorldGravity:Vector="0,0,0", ExternalForce) -> (Pose) {
    bClampLinearTranslationLimitToRefPose=true
    bOverrideWorldGravity=true
    overridePhysicsAsset=/BRCosmetics/Characters/Player/Male/Medium/Bodies/M_MED_PowerFarmer_Damage/Meshes/M_MED_PowerFarmer_Damage_Physics
    simulationSpace=BaseBoneSpace
    alphaBoolBlend {blendOption=Linear}
    baseBoneRef {boneName=pelvis}
    componentAppliedLinearAccClamp=<1000 1000 1000>
    componentLinearAccScale=<0.35 0.35 0.4>
    componentLinearVelScale=<0.2 0.2 0.2>
    overrideWorldGravity {z=-980}
   }
   n25 = RigidBody(ComponentPose=@n24, OverrideWorldGravity="0,0,0") {
    bClampLinearTranslationLimitToRefPose=true
    bOverrideWorldGravity=true
    overridePhysicsAsset=/BRCosmetics/Characters/Player/Male/Medium/Bodies/M_MED_PowerFarmer_Damage/Meshes/M_MED_PowerFarmer_Damage_Physics_SkyDive
    simulationSpace=BaseBoneSpace
    alphaBoolBlend {blendOption=Linear}
    baseBoneRef {boneName=pelvis}
    componentAppliedLinearAccClamp=<1000 1000 1000>
    componentLinearAccScale=<0.35 0.35 0.4>
    componentLinearVelScale=<0.2 0.2 0.2>
    overrideWorldGravity {z=-980}
   }
   n26 = "Constraint - dyn_jacket_bk_l_02 (PPPPP)"!(
    ComponentPose=@n25
    Alpha=1
    ConstraintWeights_0=0.2767
    ConstraintWeights_1:float=0.2695
    ConstraintWeights_2:float=0.041
    ConstraintWeights_3:float=0.2267
    ConstraintWeights_4:float=0.1861
   ) -> (Pose) {
    alphaBoolBlend {blendOption=Linear}
    boneToModify {boneName=dyn_jacket_bk_l_02}
    constraintSetup [
     {targetBone {boneName=spine_01}}
     {targetBone {boneName=spine_02}}
     {targetBone {boneName=spine_04}}
     {targetBone {boneName=pelvis}}
     {targetBone {boneName=spine_03}}
    ]
    constraintWeights [1,1,1,1,1]
   }
   n27 = "Constraint - dyn_jacket_bk_r_03 (PPPPP)"!(
    ComponentPose=@n26
    Alpha=1
    ConstraintWeights_0=0.2782
    ConstraintWeights_1:float=0.2472
    ConstraintWeights_2:float=0.044
    ConstraintWeights_3:float=0.2593
    ConstraintWeights_4:float=0.1713
   ) -> (Pose) {
    alphaBoolBlend {blendOption=Linear}
    boneToModify {boneName=dyn_jacket_bk_r_03}
    constraintSetup [
     {targetBone {boneName=spine_01}}
     {targetBone {boneName=spine_02}}
     {targetBone {boneName=spine_04}}
     {targetBone {boneName=pelvis}}
     {targetBone {boneName=spine_03}}
    ]
    constraintWeights [1,1,1,1,1]
   }
   n28 = "Constraint - dyn_jacket_fr_l_02 (PPPPP)"!(
    ComponentPose=@n27
    Alpha=1
    ConstraintWeights_0=0.2855
    ConstraintWeights_1:float=0.2856
    ConstraintWeights_2:float=0.047
    ConstraintWeights_3:float=0.1772
    ConstraintWeights_4:float=0.2047
   ) -> (Pose) {
    alphaBoolBlend {blendOption=Linear}
    alphaScaleBiasClamp {bClampResult=true, bInterpResult=true, clampMax=0.5, interpSpeedDecreasing=5, interpSpeedIncreasing=7}
    boneToModify {boneName=dyn_jacket_fr_l_02}
    constraintSetup [
     {targetBone {boneName=spine_01}}
     {targetBone {boneName=spine_02}}
     {targetBone {boneName=spine_04}}
     {targetBone {boneName=pelvis}}
     {targetBone {boneName=spine_03}}
    ]
    constraintWeights [1,1,1,1,1]
   }
   n29 = "Constraint - dyn_jacket_bk_l_03 (PPP)"!(ComponentPose=@n28, Alpha=1, ConstraintWeights_0=0.2646, ConstraintWeights_1:float=0.7224, ConstraintWeights_2:float=0.013) -> (Pose) {
    alphaBoolBlend {blendOption=Linear}
    alphaScaleBiasClamp {bClampResult=true, bInterpResult=true, clampMax=0.8, interpSpeedDecreasing=5, interpSpeedIncreasing=7}
    boneToModify {boneName=dyn_jacket_bk_l_03}
    constraintSetup [{targetBone {boneName=spine_01}}, {targetBone {boneName=pelvis}}, {targetBone {boneName=spine_03}}]
    constraintWeights [1,1,1]
   }
   n30 = "Constraint - dyn_jacket_bk_r_03 (PPPPPP)"!(
    ComponentPose=@n29
    Alpha=1
    ConstraintWeights_0=0.2695
    ConstraintWeights_1:float=0.0802
    ConstraintWeights_2:float=0.0074
    ConstraintWeights_3:float=0.592
    ConstraintWeights_4:float=0.0488
    ConstraintWeights_5:float=0.0021
   ) -> (Pose) {
    alphaBoolBlend {blendOption=Linear}
    alphaScaleBiasClamp {bClampResult=true, bInterpResult=true, clampMax=0.8, interpSpeedDecreasing=5, interpSpeedIncreasing=7}
    boneToModify {boneName=dyn_jacket_bk_r_03}
    constraintSetup [
     {targetBone {boneName=spine_01}}
     {targetBone {boneName=spine_02}}
     {targetBone {boneName=spine_04}}
     {targetBone {boneName=pelvis}}
     {targetBone {boneName=spine_03}}
     {targetBone {boneName=thigh_twist_01_r}}
    ]
    constraintWeights [1,1,1,1,1,1]
   }
   Output Pose(Result=Component To Local(ComponentPose=@n30)) {layerGroup=None}
  ]}
 EventGraph {nodes [
   Event BlueprintUpdateAnimation()
   > n1 = Sequence!(execute) -> (then_0, then_1, then_2:Exec, then_3:Exec, then_4:Exec)
   > n2 = Gravity Override State Function DEPRECATED(CharacterPartAnimInstance=Self-Reference(), GameTimeInSeconds=GetGameTimeInSeconds(), GravityOverrideParams=Get Gravity Override Params())
   > Set Gravity Override(Gravity Override=@n2)
   Set Crouch(execute=@n1.then_1, Crouch:double=SelectFloat(A=1, B=0, bPickA=Get bIsCrouching())) -> (Output_Get:double)
   n3 = NOT Boolean(A=Get bIsSkyDiving())
   n4 = OR Boolean(A=Get bIsJumping(), B=Get bIsFalling())
   > Set Jump(Jump:double=SelectFloat(A=1, B=0, bPickA=AND Boolean(A=@n4, B=@n3))) -> (Output_Get:double)
   n5 = NOT Boolean(A=OR Boolean(A=Get bIsJumping(), B=Get bIsFalling()))
   n6 = Get bIsMoving2D()
   > Set moving(moving:double=SelectFloat(A=1, B=0, bPickA=AND Boolean(A=@n6, B=@n5))) -> (Output_Get:double)
   n7 = Get bIsCrouching()
   > Set ADS_Stand(ADS_Stand:double=SelectFloat(A=1, B=0, bPickA=NOT Boolean(A=@n7))) -> (Output_Get:double)
   n8 = CheckForBackbling()
   > Set WearingBackbling(WearingBackbling:bool=OR Boolean!(A=@n8, B=@n8.bIsBackblingACape, C:bool=@n8.bIsBackblingLong) -> (ReturnValue)) -> (Output_Get:bool)
   > n9 = Branch(Condition=Get bIsSkyDiving())
   > Set Skydive_off(Skydive_off:double=0) -> (Output_Get:double)
   Set Skydive_off(execute=@n9.else, Skydive_off=1)
   n10 = Branch(execute=@n1.then_2, Condition=Get bIsSliding())
   > n11 = Set Left_Calf(Left_Calf:double=GetAlphaFromReferenceTransformDelta(SkeletalMeshComponent=GetPartSkeletalMeshComponent(), BoneName=calf_l, BaseName=thigh_l, InputRange=(X=0,Y=1), TwistAxis="1,0,0").RawOutput) -> (Output_Get:double)
   > n12 = Branch(Condition=NOT Boolean(A="float > float"(A=@n11, B=100)))
   > Set Base_Slide(Base_Slide:double=1) -> (Output_Get:double)
   > Set Knee_Slide(Knee_Slide:double=0) -> (Output_Get:double)
   Set Base_Slide(execute=@n12.else, Base_Slide=0)
   > Set Knee_Slide(Knee_Slide=1)
   Set Base_Slide(execute=@n10.else, Base_Slide=0)
   > Set Knee_Slide(Knee_Slide=0)
   n13 = Cast To FortPlayerAnimInstance(execute=@n1.then_3, Object=GetAnimInstance(self=GetOwnerSkeletalMeshComponent()))
   > Set Pistol(Pistol:double=0) -> (Output_Get:double)
   > Set Rifle(Rifle:double=0) -> (Output_Get:double)
   > n14 = Switch on EFortWeaponCoreAnimation!(execute, Selection:EFortWeaponCoreAnimation Enum=Get WeaponCoreAnim(self=@n13)) -> (
    Melee:Exec
    Pistol:Exec
    Shotgun:Exec
    PaperBlueprint:Exec
    Rifle:Exec
    MeleeOneHand:Exec
    MachinePistol:Exec
    RocketLauncher:Exec
    GrenadeLauncher:Exec
    GoingCommando:Exec
    AssaultRifle:Exec
    TacticalShotgun:Exec
    SniperRifle:Exec
    TrapPlacement:Exec
    ShoulderLauncher:Exec
    AbilityDecoTool:Exec
    Crossbow:Exec
    C4:Exec
    RemoteControl:Exec
    DualWield:Exec
    AR_BullPup:Exec
    AR_ForwardGrip:Exec
    MedPackPaddles:Exec
    SMG_P90:Exec
    AR_DrumGun:Exec
    Consumable_Small:Exec
    Consumable_Large:Exec
    Balloon:Exec
    MountedTurret:Exec
    CreativeTool:Exec
    ExplosiveBow:Exec
    AshtonIndigo:Exec
    AshtonChicago:Exec
    MeleeDualWield:Exec
    Unarmed:Exec
   )
   Branch(execute=[
     @n14.AssaultRifle
     @n14.Crossbow
     @n14.GrenadeLauncher
     @n14.Rifle
     @n14.RocketLauncher
     @n14.Shotgun
     @n14.SniperRifle
     @n14.TacticalShotgun
    ])
   > Set Rifle(Rifle=1)
   Branch(execute=[@n14.GoingCommando, @n14.MachinePistol, @n14.Pistol])
   > Set Pistol(Pistol=1)
   n15 = PlayerMovement(execute=@n1.then_4, AnimInstance=Self-Reference(), XDotProductCoefficient=100, YDotProductCoefficient=100, ZDotProductCoefficient=100)
   n16 = BreakVector(InVec=MakeVector(X=@n15, Y=@n15.YMovement))
   n17 = "float < float"(A=@n16, B=-50)
   > Set Left(Left:double=ClampAxis!(Angle:float="To Float (Boolean)"(InBool=@n17)) -> (ReturnValue:float)) -> (Output_Get:double)
   n18 = "float > float"(A=@n16, B=50)
   > Set Right(Right:double=ClampAxis(Angle="To Float (Boolean)"(InBool=@n18))) -> (Output_Get:double)
   n19 = "float > float"(A=@n16.Y, B=40)
   > Set Forward(Forward:double=ClampAxis(Angle="To Float (Boolean)"(InBool=@n19))) -> (Output_Get:double)
   n20 = "float < float"(A=@n16.Y, B=-40)
   > Set Backward(Backward:double=ClampAxis(Angle="To Float (Boolean)"(InBool=@n20))) -> (Output_Get:double)
   AND Boolean(A=Get bIsTargeting(), B=false)
  ]}
]
variables [
 ADS_Stand {type=number}
 Backward {type=number}
 Base_Slide {type=number}
 Crouch {type=number}
 Forward {type=number}
 Gravity Override {type=Vector}
 Gravity Override Params {type=GravityOverrideParamsStruct}
 isAiming_Pistol {type=boolean}
 isAimingRifle {type=number}
 Jump {type=number}
 Knee_Slide {type=number}
 Left {type=number}
 Left_Calf {type=number}
 moving {type=number}
 Pistol {type=number}
 Rifle {type=number}
 Right {type=number}
 Skydive_off {type=number}
 WearingBackbling {type=boolean}
]
